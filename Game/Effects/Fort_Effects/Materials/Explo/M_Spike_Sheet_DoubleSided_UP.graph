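# Translucent, unlit, two-sided spike material for explosion particles, usable on Niagara sprites and mesh particles.
# A floor-based V mask splits the UVs, showing a rotating spike texture below 0.5 and a cropped spike-bottom texture above.
# Dynamic parameter Param1 drives the rotation.
# That mask, depth fade and particle alpha form the clamped opacity, while particle color feeds MF_FX_EmissiveAdjustments.
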Material /Game/Effects/Fort_Effects/Materials/Explo/M_Spike_Sheet_DoubleSided_UP
graph {
 properties {
  bUsedWithMeshParticles=true
  bUsedWithNiagaraMeshParticles=true
  bUsedWithNiagaraSprites=true
  blendMode=BLEND_Translucent
  pixelDepthOffsetMode=PDOM_Legacy
  shadingModel=MSM_Unlit
  twoSided=true
  usedShadingModels=Unlit
 }
 nodes [
  Add(A=TextureCoordinate() {vTiling=2}, B=AppendVector(A=Constant(), B=Constant() {r=-1}))
  n1 = ParticleColor()
  n2 = TextureCoordinate()
  n3 = VectorParameter() {parameterName=Crop, defaultValue=<rgba 0 0 1 0.25>}
  n4 = AppendVector(A=@n3.R, B=@n3.R)
  n5 = AppendVector(A=@n3.B, B=@n3.B)
  n6 = DynamicParameter() {defaultValue=<rgba 0 0 0 1>}
  n7 = Add(A=ComponentMask(None=TextureCoordinate()) {g=true}) {constB=0.5}
  n8 = LinearInterpolate(
   A=TextureSample(UVs=MaterialFunctionCall!(UVs=@n2, TextureIn, UpperLeftCorner=@n4, LowerRightCorner=@n5, ExportFloat4) {materialFunction=/Engine/Functions/Engine_MaterialFunctions01/Texturing/TextureCropping}.Crop UVs) {samplerSource=SSM_Clamp_WorldGroupSettings, texture=/Game/Effects/Fort_Effects/Textures/Explo/T_Spike_Bottom}.RGB
   B=TextureSample(UVs=Rotator(Time=@n6.Param1)) {texture=/Game/Effects/Fort_Effects/Textures/Explo/T_Spike_01_Rotate}.R
   Alpha=OneMinus(None=Floor(None=@n7))
  )
  n9 = DepthFade() {fadeDistanceDefault=300}
  Clamp(None=Multiply(A=@n1.A, B=Multiply(A=@n8, B=@n9)))
  MaterialFunctionCall(Emissive=@n1.RGB) {materialFunction=&MF_FX_EmissiveAdjustments}
 ]
}
